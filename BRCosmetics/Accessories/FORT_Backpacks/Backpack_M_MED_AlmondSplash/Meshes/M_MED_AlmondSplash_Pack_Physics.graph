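# Physics asset for the AlmondSplash backpack cosmetic, covering its dangling parts: stick arms, jacket flaps, collar fur, hat, necklace and string light.
# Small spheres mark base bones, while capsules, boxes and larger spheres approximate the jointed parts.
# Each part links to its own base with tight swing and twist limits, with hat and string light forming multi-link chains.
# No constraint connects those bases to each other or to the backpack base body.
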PhysicsAsset /BRCosmetics/Accessories/FORT_Backpacks/Backpack_M_MED_AlmondSplash/Meshes/M_MED_AlmondSplash_Pack_Physics
bodies {
 dyn_backpack_base [{shapeType=Sphere, radius=0.5, shapeName=dyn_backpack_base_sphere}]
 dyn_collar_fur_l_base [{shapeType=Sphere, radius=0.5, shapeName=dyn_collar_fur_l_base_sphere}]
 dyn_collar_fur_l_parent [{shapeType=Sphere, radius=3, shapeName=dyn_collar_fur_l_parent_sphere, center=<0 0 -1.139>}]
 dyn_collar_fur_r_base [{shapeType=Sphere, radius=0.5, shapeName=dyn_collar_fur_r_base_sphere}]
 dyn_collar_fur_r_parent [{shapeType=Sphere, radius=3, shapeName=dyn_collar_fur_r_parent_sphere, center=<0.1452 0.3063 -0.6995>}]
 dyn_hat_01 [{shapeType=Capsule, length=2, radius=1, shapeName=dyn_hat_01_capsule, center=<1.307 0 0>, rotation=<rot 90 0 0>}]
 dyn_hat_base [{shapeType=Sphere, radius=0.5, shapeName=dyn_hat_base_sphere}]
 dyn_hat_parent [{shapeType=Capsule, length=2, radius=1, shapeName=dyn_hat_parent_capsule, center=<1.2811 0 0>, rotation=<rot 90 0 0>}]
 dyn_jacket_l_base [{shapeType=Sphere, radius=0.5, shapeName=dyn_jacket_l_base_sphere}]
 dyn_jacket_l_parent [{shapeType=Box, extentX=4.5, extentY=5.5, extentZ=1.5, shapeName=dyn_jacket_l_parent_box, center=<2.5615 0 0>}]
 dyn_jacket_r_base [{shapeType=Sphere, radius=0.5, shapeName=dyn_jacket_r_base_sphere}]
 dyn_jacket_r_parent [{shapeType=Box, extentX=4.5, extentY=5.5, extentZ=1.5, shapeName=dyn_jacket_r_parent_box, center=<1.9137 0 0>}]
 dyn_necklace_base [{shapeType=Sphere, radius=0.5, shapeName=dyn_necklace_base_sphere}]
 dyn_necklace_parent [{shapeType=Capsule, length=1.5, radius=1.5, shapeName=dyn_necklace_parent_capsule, center=<1.2477 0 0>, rotation=<rot 90 0 0>}]
 dyn_stick_arm_l_base [{shapeType=Sphere, radius=0.5, shapeName=dyn_stick_arm_l_base_sphere}]
 dyn_stick_arm_l_parent [{shapeType=Capsule, length=7.1486, radius=2.7929, shapeName=dyn_stick_arm_l_parent_capsule, center=<4.8414 0 0>, rotation=<rot 90 0 0>}]
 dyn_stick_arm_r_base [{shapeType=Sphere, radius=0.5, shapeName=dyn_stick_arm_r_base_sphere}]
 dyn_stick_arm_r_parent [{shapeType=Capsule, length=5.153, radius=3, shapeName=dyn_stick_arm_r_parent_capsule, center=<4.7904 0 0>, rotation=<rot 90 0 0>}]
 dyn_string_light_01 [{shapeType=Capsule, length=3, radius=1, shapeName=dyn_string_light_01_capsule, center=<2.1444 0 0>, rotation=<rot 90 0 0>}]
 dyn_string_light_02 [{shapeType=Capsule, length=3, radius=1, shapeName=dyn_string_light_02_capsule, center=<2.1444 0 0>, rotation=<rot 90 0 0>}]
 dyn_string_light_base [{shapeType=Sphere, radius=0.5, shapeName=dyn_string_light_base_sphere}]
 dyn_string_light_parent [{shapeType=Capsule, length=3, radius=1, shapeName=dyn_string_light_parent_capsule, center=<2.1444 0 0>, rotation=<rot 90 0 0>}]
}
constraints [
 {bone1Name=dyn_stick_arm_r_parent, bone2Name=dyn_stick_arm_r_base, swing1LimitDegrees=2, swing2LimitDegrees=10, twistLimitDegrees=3}
 {bone1Name=dyn_stick_arm_l_parent, bone2Name=dyn_stick_arm_l_base, swing1LimitDegrees=2, swing2LimitDegrees=10, twistLimitDegrees=3}
 {bone1Name=dyn_jacket_l_parent, bone2Name=dyn_jacket_l_base, swing1LimitDegrees=10, swing2LimitDegrees=10, twistLimitDegrees=3}
 {bone1Name=dyn_jacket_r_parent, bone2Name=dyn_jacket_r_base, swing1LimitDegrees=10, swing2LimitDegrees=10, twistLimitDegrees=3}
 {bone1Name=dyn_collar_fur_r_parent, bone2Name=dyn_collar_fur_r_base, swing1LimitDegrees=10, swing2LimitDegrees=10, twistLimitDegrees=3}
 {bone1Name=dyn_collar_fur_l_parent, bone2Name=dyn_collar_fur_l_base, swing1LimitDegrees=10, swing2LimitDegrees=10, twistLimitDegrees=3}
 {bone1Name=dyn_hat_01, bone2Name=dyn_hat_parent, swing1LimitDegrees=5, swing2LimitDegrees=5, twistLimitDegrees=3}
 {bone1Name=dyn_hat_parent, bone2Name=dyn_hat_base, swing1LimitDegrees=5, swing2LimitDegrees=5, twistLimitDegrees=3}
 {bone1Name=dyn_string_light_parent, bone2Name=dyn_string_light_base, swing1LimitDegrees=5, swing2LimitDegrees=5, twistLimitDegrees=3}
 {bone1Name=dyn_string_light_01, bone2Name=dyn_string_light_parent, swing1LimitDegrees=5, swing2LimitDegrees=5, twistLimitDegrees=3}
 {bone1Name=dyn_string_light_02, bone2Name=dyn_string_light_01, swing1LimitDegrees=5, swing2LimitDegrees=5, twistLimitDegrees=3}
 {bone1Name=dyn_necklace_parent, bone2Name=dyn_necklace_base, swing1LimitDegrees=3, swing2LimitDegrees=3, twistLimitDegrees=2}
]
